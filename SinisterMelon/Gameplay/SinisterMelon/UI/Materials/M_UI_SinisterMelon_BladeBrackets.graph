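# UI material that draws mirrored blade-shaped brackets as a fill meter driven by PercentFill, with glow and flame effects.
# A mirrored SDF texture gives the blade silhouette, intersected with a circle for its edge, while an angle gradient splits fill from background.
# Distorted panning fire noise and scrolling embers animate the fill outline.
# Colour layers stack by lerps and linear dodge, then multiply by accumulated opacity for alpha-composite blending.
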Material /SinisterMelon/Gameplay/SinisterMelon/UI/Materials/M_UI_SinisterMelon_BladeBrackets
graph {
 properties {bAutomaticallySetUsageInEditor=false, blendMode=BLEND_AlphaComposite, materialDomain=MD_UI, usedShadingModels=Default Lit}
 nodes [
  n1 = VectorParameter() {parameterName=Blade Background - Color, defaultValue=<rgba 0.0052 0.0091 0.0262 1>}
  LinearInterpolate(
   A=@n1.RGB
   B=VectorParameter() {parameterName="Blade Fill, Outer Glow - Color", defaultValue=<rgba 1 0.8714 0.3813 1>}.RGB
   Alpha=Ceil(None=NamedRerouteUsage())
  )
  n2 = OneMinus(None=NamedRerouteUsage())
  n3 = ScalarParameter() {defaultValue=4, group=FlameMask, parameterName=OutlineSmooth}
  n4 = Multiply(A=@n3)
  n5 = Multiply(A=@n4) {constB=0.1}
  n6 = Add(A=Time(), B=ScalarParameter() {parameterName=TimeOffset})
  n7 = MaterialFunctionCall(Value=@n6) {materialFunction=&LinearSine}
  n8 = Constant()
  n9 = Add(A=ScalarParameter() {defaultValue=1, parameterName=TopAdjust}, B=ScalarParameter() {parameterName=ActivateAdjust})
  n10 = ComponentMask(None=Add(A=TextureCoordinate(), B=AppendVector(A=@n8, B=@n9))) {g=true}
  n11 = TextureCoordinate()
  n12 = Multiply(A=TextureCoordinate(), B=ScalarParameter() {defaultValue=1, parameterName=PatternScale})
  n13 = MaterialFunctionCall(Float2=MaterialFunctionCall() {materialFunction=&GetUserInterfaceUV}.Pixel Size) {materialFunction=&BreakOutFloat2Components}
  n14 = VectorParameter() {parameterName=InterfaceScale, defaultValue=<rgba 1 1 0 1>}
  n15 = Multiply(A=AppendVector(A=Divide(A=@n13.R, B=@n13.R), B=Constant() {r=1}), B=AppendVector(A=@n14.R, B=@n14.R)) {desc=Scale X to fit}
  n16 = Multiply(A=@n12, B=@n15)
  n17 = ScalarParameter() {parameterName=ScreenAlignedUVs}
  n18 = VectorParameter() {parameterName=Speed2, defaultValue=<rgba 0 0.5 0 1>}
  n19 = TextureSampleParameter2D(UVs=Panner(Coordinate=LinearInterpolate(A=@n11, B=@n16, Alpha=@n17), Time=@n6, Speed=AppendVector(A=@n18.R, B=@n18.R)) {speedY=0.5}) {parameterName=FireTexture2, texture=/Game/Effects/Fort_Effects/Textures/Fire/T_FireNoise_02}
  n20 = Multiply(A=@n12)
  n21 = VectorParameter() {parameterName=Speed1, defaultValue=<rgba 0 0.2 0 1>}
  n22 = Panner(
   Coordinate=MaterialFunctionCall!(
    HeightTexture=@n19.R
    TextureCoordinates=@n20
    ReferencePlane
    HeightRatioInput=ScalarParameter() {defaultValue=2, parameterName=FlameDistort}
    CameraVector
   ) {materialFunction=/Engine/Functions/Engine_MaterialFunctions02/Utility/BumpOffset_Advanced.BumpOffset_advanced}.Undeformed UVs
   Time=@n6
   Speed=AppendVector(A=@n21.R, B=@n21.R)
  ) {speedY=0.2}
  n23 = MaterialFunctionCall(
   UVs=@n20
   CenterPosition=Constant2Vector() {g=1, r=0.5}
   Radius=LinearInterpolate(A=Constant() {r=1}, B=Constant() {r=2}, Alpha=@n7.Linear Sine)
   Density=Constant() {r=0.7164}
  ) {materialFunction=&RadialGradientExponential}
  n24 = LinearInterpolate(
   B=Power(Base=@n19.R, Exp=ScalarParameter() {defaultValue=-0.5, parameterName=FlameExp}) {constExponent=-0.75}
   Alpha=Power(
    Base=TextureSampleParameter2D(UVs=@n22) {parameterName=FireTexture1, texture=/Game/Global/Textures/Default/Noise/T_FireNoise_01}.R
    Exp=Multiply(A=@n23.RadialGradientExponential) {constB=-1}
   ) {constExponent=5}
  )
  n25 = MaterialFunctionCall() {materialFunction=&GetUserInterfaceUV}
  n26 = VectorParameter() {group=TopGradient, parameterName=TopGradientScale, defaultValue=<rgba 0 0 0 1>}
  n27 = AppendVector(A=@n26.R, B=@n26.R)
  n28 = VectorParameter() {group=TopGradient, parameterName=TopGradientPosition, defaultValue=<rgba 0.5 0 0 0>}
  n29 = OneMinus(None=MaterialFunctionCall(
    UVs=MaterialFunctionCall(UVs=@n25.Normalized UV, Texture Scale=@n27) {materialFunction=&ScaleUVsByCenter}.UVs
    CenterPosition=AppendVector(A=@n28.R, B=@n28.R)
    Radius=@n28.B
    Density=@n28.B
   ) {materialFunction=&RadialGradientExponential}.RadialGradientExponential)
  n30 = MaterialFunctionCall(Texture Scale=Constant2Vector() {g=2, r=3}) {materialFunction=&ScaleUVsByCenter}
  n31 = VectorParameter() {parameterName=FlameHeight, defaultValue=<rgba 0 0 0 1>}
  n32 = AppendVector(A=@n31.R, B=@n31.R)
  n33 = VectorParameter() {parameterName=FlameCenterPosition, defaultValue=<rgba 0.5 1 0.8043 2.3384>}
  n34 = ScalarParameter() {defaultValue=1, group=DFMask, parameterName=DFScale}
  n35 = ScalarParameter() {defaultValue=1, parameterName=DFActivate}
  n36 = TextureSampleParameter2D(UVs=MaterialFunctionCall!(
    Texture Scale=Multiply(A=@n34, B=@n35)
    UVs=MaterialFunctionCall() {materialFunction=&GetUserInterfaceUV}.Normalized UV
    Offset XY Texturesize Z=VectorParameter() {group=DFMask, parameterName=DFScalePivot, defaultValue=<rgba 0.5 1 1 0>}.RGB
   ) {materialFunction=/Engine/Functions/Engine_MaterialFunctions01/Texturing/ScaleUVsAroundPoint}.UVs) {
   group=DFMask
   parameterName=DFMaskTexture
   samplerSource=SSM_Clamp_WorldGroupSettings
   texture=/Game/Athena/HUD/Materials/Score/T_UI_Square_DF
  }
  n37 = LinearInterpolate(
   A=MaterialFunctionCall(UVs=Add(A=@n30.UVs, B=@n32), CenterPosition=@n33.RGB, Radius=@n33.RGB, Density=@n33.RGB, Invert Density=StaticBool() {value=true}) {materialFunction=&RadialGradientExponential}.RadialGradientExponential
   B=@n36.R
   Alpha=ScalarParameter() {defaultValue=0.05, group=DFMask, parameterName=DFMask}
  )
  n38 = NamedRerouteDeclaration(None=Multiply(A=Multiply(A=@n10, B=@n24), B=Multiply(A=@n29, B=@n37))) {name=V_FlameMask, nodeColor=<rgba 0 0 0 1>}
  n39 = MaterialFunctionCall(
   Alpha=@n38
   Min=ScalarParameter() {defaultValue=0.1, parameterName=MaskMin}
   Max=ScalarParameter() {defaultValue=0.76, parameterName=MaskMax}
  ) {materialFunction=&SmoothStep}
  MaterialFunctionCall(Alpha=@n39.Result, Min=@n2, Max=Add(A=@n2, B=LinearInterpolate(A=@n4, B=@n5, Alpha=@n7.Linear Sine)) {constB=0.01}) {desc=Flame Edge Ember Mask, materialFunction=&SmoothStep}
  n40 = NamedRerouteUsage()
  Multiply(A=@n40, B=OneMinus(None=NamedRerouteUsage()))
  n41 = VectorParameter() {parameterName="Blade Fill, Leading Edge - Color", defaultValue=<rgba 1 0.8714 0.3813 1>}
  n42 = Multiply(A=NamedRerouteUsage(), B=ScalarParameter() {parameterName=Blade Flash - Opacity})
  n43 = Multiply(
   A=NamedRerouteUsage()
   B=LinearInterpolate(B=ScalarParameter() {defaultValue=0.8, parameterName=Blade Fill Leading Edge - Opacity}, Alpha=NamedRerouteUsage())
  )
  n44 = Multiply(A=NamedRerouteUsage(), B=ScalarParameter() {defaultValue=0.8, parameterName=Blade Edge Fill - Opacity})
  n45 = VectorParameter() {parameterName=Blade Fill - Color, defaultValue=<rgba 1 0.6724 0.0578 1>}
  n46 = Multiply(A=NamedRerouteUsage(), B=ScalarParameter() {defaultValue=0.88, parameterName=Blade Fill - Opacity})
  n47 = Multiply(A=NamedRerouteUsage(), B=ScalarParameter() {defaultValue=0.005, parameterName=Blade Edge Background - Opacity})
  n48 = NamedRerouteUsage()
  n49 = LinearInterpolate(
   A=LinearInterpolate(
    A=VectorParameter() {parameterName="Blade Fill, Outer Glow - Color", defaultValue=<rgba 1 0.8714 0.3813 1>}.RGB
    B=@n1.RGB
    Alpha=Ceil(None=@n48)
   )
   B=VectorParameter() {parameterName=Blade Edge Background - Color, defaultValue=<rgba 1 1 1 0>}.RGB
   Alpha=@n47
  )
  n50 = MaterialFunctionCall!(
   Blend=LinearInterpolate(
    A=LinearInterpolate(A=@n49, B=@n45.RGB, Alpha=@n46)
    B=VectorParameter() {parameterName=Blade Edge Fill - Color, defaultValue=<rgba 1 1 0.4125 1>}.RGB
    Alpha=@n44
   )
   Base=LinearInterpolate(B=@n41.RGB, Alpha=@n43)
  ) {materialFunction=/Engine/Functions/Engine_MaterialFunctions03/Blends/Blend_LinearDodge}
  n51 = NamedRerouteUsage()
  n52 = Multiply(A=NamedRerouteUsage(), B=ScalarParameter() {defaultValue=0.8, parameterName=Blade Outline Glow/Embers - Opacity})
  n53 = ScalarParameter() {defaultValue=0.8, parameterName="Blade Fill, Outer Glow - Opacity"}
  n54 = LinearInterpolate(
   A=LinearInterpolate(Alpha=Multiply(A=@n40, B=@n53))
   Alpha=Multiply(A=NamedRerouteUsage(), B=ScalarParameter() {defaultValue=0.32, parameterName=Blade Background - Opacity})
  )
  n55 = LinearInterpolate(A=LinearInterpolate(A=LinearInterpolate(A=@n54, Alpha=@n47), Alpha=@n46), Alpha=@n44)
  n56 = LinearInterpolate(A=LinearInterpolate(A=LinearInterpolate(A=@n55, Alpha=@n43), Alpha=@n42), Alpha=@n52)
  Multiply(A=LinearInterpolate(A=LinearInterpolate(A=@n50.Result, B=@n41.RGB, Alpha=@n42), B=Add(A=@n50.Result, B=@n51), Alpha=@n52), B=@n56)
  n57 = OneMinus(None=NamedRerouteUsage())
  n58 = NamedRerouteUsage()
  n59 = NamedRerouteDeclaration(None=Saturate(None=ScalarParameter() {defaultValue=0.5, parameterName=PercentFill})) {name=V_PercentFill, nodeColor=<rgba 0 0 0 1>}
  n60 = OneMinus(None=LinearInterpolate(Alpha=Subtract(A=@n59) {constB=-0.02}) {constA=0.634, constB=0.829})
  n61 = ScalarParameter() {defaultValue=0.02, parameterName=Flame Edge - Fill Softness}
  n62 = Add(A=@n61) {constB=0.026}
  n63 = MaterialFunctionCall!(Repeat Count=Constant2Vector() {g=1, r=2}, texcoord=TextureCoordinate()) {materialFunction=/Game/Environments/Papaya/Materials/MaterialFunctions/MF_UVMirror}
  n64 = Constant2Vector() {g=0.4, r=0.8}
  n65 = MaterialFunctionCall!(
   UVs=Add(
    A=MaterialFunctionCall(UVs=@n63.Result, Texture Scale=@n64) {materialFunction=&ScaleUVsByCenter}.UVs
    B=Constant2Vector() {g=0.06, r=-0.5}
   )
   Rotation=Constant()
  ) {materialFunction=/FortUILibrary/MaterialFunctions/Gradients/MF_UI_AngleGradient}
  n66 = SmoothStep(Min=Subtract(A=@n60, B=@n62), Max=Add(A=@n60, B=@n62), Value=@n65.Result) {constMin=0.5}
  n67 = Multiply(A=Subtract(A=@n66) {constB=0.2}) {constB=0.5}
  n68 = Multiply(A=NamedRerouteUsage()) {constB=2}
  n69 = Multiply(A=LinearInterpolate(B=Min(A=@n58, B=@n67), Alpha=Saturate(None=@n68))) {constB=20}
  n70 = Constant() {r=3.6}
  n71 = Constant() {r=0.4}
  n72 = Constant() {r=0.2}
  n73 = Constant() {r=10}
  n74 = OneMinus(None=Saturate(None=MaterialFunctionCall!(DistanceField=@n69, Threshold=@n70, Thickness=@n71, Alignment=@n72, Smoothing=@n73) {materialFunction=/FortUILibrary/MaterialFunctions/Shapes/Modifiers/MF_UI_ShapeOutline}.SDFOutline))
  n75 = NamedRerouteUsage()
  NamedRerouteDeclaration(None=Multiply(A=SmoothStep(Min=@n57, Value=@n74), B=OneMinus(None=@n75))) {name=V_Blade_Fill_Outline_SDF, nodeColor=<rgba 0 0 0 1>}
  n76 = OneMinus(None=LinearInterpolate(Alpha=Subtract(A=@n59) {constB=0.14}) {constA=0.634, constB=0.829})
  n77 = Add(A=@n61) {constB=0.04}
  n78 = Subtract(A=@n76, B=@n77)
  n79 = Add(A=@n76, B=@n77)
  n80 = MaterialFunctionCall!(
   DistanceFieldA=OneMinus(None=NamedRerouteUsage())
   DistanceFieldB=OneMinus(None=SmoothStep(Min=@n78, Max=@n79, Value=@n65.Result) {constMin=0.5})
   BlendAmount=Constant() {r=0.05}
  ) {materialFunction=/FortUILibrary/MaterialFunctions/Shapes/Modifiers/MF_UI_SDF_Intersection}
  n81 = SmoothStep(
   Min=@n57
   Value=MaterialFunctionCall(DistanceField=OneMinus(None=@n80.Smooth) {desc=Blade Fill Glow}, Threshold=Constant() {r=0.12}, Smoothing=Constant() {r=0.2}) {materialFunction=&MF_UI_AntiAliasing}.Outer
  )
  NamedRerouteDeclaration(None=Multiply(A=NamedRerouteDeclaration(None=@n81) {name=V_Blade_Fill_Outline_Ember_Mask, nodeColor=<rgba 0 0 0 1>}, B=NamedRerouteUsage())) {name=V_Blade_Fill_Glow, nodeColor=<rgba 0 0 0 1>}
  n82 = NamedRerouteDeclaration(None=MaterialFunctionCall(
    DistanceField=OneMinus(None=@n65.Result)
    Threshold=LinearInterpolate(Alpha=@n59) {constA=0.634, constB=0.829}
    Smoothing=Constant() {r=0.002}
   ) {desc=Progress Radial Fill, materialFunction=&MF_UI_AntiAliasing}.Inner) {name=V_ProgressFill, nodeColor=<rgba 0 0 0 1>}
  NamedRerouteDeclaration(None=OneMinus(None=@n82)) {name=V_ProgressFill_Inv, nodeColor=<rgba 0 0 0 1>}
  n83 = OneMinus(None=@n66)
  n84 = NamedRerouteUsage()
  NamedRerouteDeclaration(None=Multiply(A=Multiply(A=@n83, B=@n84) {desc="Blade Fill, Leading Edge"}) {constB=2}) {name=V_Blade_Fill_Leading_Edge, nodeColor=<rgba 0 0 0 1>}
  NamedRerouteDeclaration(None=Multiply(A=@n59, B=ScalarParameter() {defaultValue=0.5, parameterName=Fill Effect Multiplier})) {name=V_PercentFill_Adjusted, nodeColor=<rgba 0 0 0 1>}
  n85 = Constant2Vector() {g=0.5, r=0.5}
  n86 = MaterialFunctionCall!(UV=TextureCoordinate(), Mirror Y, Swizzle UV) {materialFunction=/FortUILibrary/MaterialFunctions/Basics/MF_UI_Symmetry_2x1_New}
  n87 = MaterialFunctionCall(UVs=@n86.Flipped Symmetry, Texture Scale=@n85) {materialFunction=&ScaleUVsByCenter}
  n88 = Multiply(
   A=TextureProperty(Tex=TextureObject() {samplerType=SAMPLERTYPE_Grayscale, texture=/SinisterMelon/Gameplay/SinisterMelon/UI/Textures/T_UI_SinisterMelon_Blade_Right})
   B=Constant2Vector() {g=0.25, r=0.5}
  )
  n89 = Constant2Vector() {g=10, r=10}
  n90 = NamedRerouteDeclaration(None=TextureSample(UVs=MaterialFunctionCall(UV=@n87.UVs, TextureSize=@n88, ContainerSize=@n89) {materialFunction=&MF_UI_UVScaling}.Scaled to Fit) {samplerType=SAMPLERTYPE_Grayscale, texture=&T_UI_SinisterMelon_Blade_Right}.R) {name=V_Blade_SDF, nodeColor=<rgba 0 0 0 1>}
  n91 = NamedRerouteDeclaration(None=MaterialFunctionCall(DistanceField=@n90, Threshold=Constant() {r=0.188}, Smoothing=Constant() {r=0.02}) {desc=Blade Silhouette, materialFunction=&MF_UI_AntiAliasing}.Outer) {name=V_Blade, nodeColor=<rgba 0 0 0 1>}
  NamedRerouteDeclaration(None=Multiply(A=OneMinus(None=@n82), B=@n91) {desc=Blade Fill}) {name=V_Blade_Back, nodeColor=<rgba 0 0 0 1>}
  n92 = VectorParameter() {parameterName=Color2, defaultValue=<rgba 0.3281 0.0154 0 1>}
  n93 = VectorParameter() {parameterName=Color1, defaultValue=<rgba 1 0.8698 0.141 1>}
  n94 = Power(Base=@n38, Exp=ScalarParameter() {defaultValue=0.75, parameterName=Firepower})
  n95 = Power(Base=@n39.Result)
  n96 = Clamp(None=@n95)
  n97 = LinearInterpolate(A=@n92.RGB, B=@n93.RGB, Alpha=Multiply(A=@n94, B=@n96))
  n98 = TextureCoordinate()
  n99 = VectorParameter() {parameterName=ColorSwitch}
  n100 = Add(A=Power(Base=ComponentMask(None=@n98) {g=true}, Exp=@n99.G), B=@n99.B)
  n101 = VectorParameter() {group=Embers, parameterName=EmberSpeed, defaultValue=<rgba 0 0.5 0 1>}
  n102 = Panner(Coordinate=Multiply(A=TextureCoordinate(), B=@n15), Speed=AppendVector(A=@n101.R, B=@n101.R))
  n103 = LinearInterpolate(
   A=VectorParameter() {group=Embers, parameterName=EmberColor, defaultValue=<rgba 0.3281 0.0154 0 1>}.RGB
   B=VectorParameter() {group=Embers, parameterName=EmberColor2, defaultValue=<rgba 1 0.8698 0.141 1>}.RGB
   Alpha=@n94
  )
  n104 = ScalarParameter() {defaultValue=10, group=Embers, parameterName=EmberBoost}
  n105 = Multiply(
   A=Multiply(
    A=TextureSampleParameter2D(UVs=@n102) {group=Embers, parameterName=EmberTexture, texture=/Game/Global/Textures/Default/PointSplatter/T_UI_StarBurst_Scatter}.R
    B=Multiply(A=@n103, B=@n104)
   )
   B=@n38
  )
  n106 = Multiply(
   A=ConstantBiasScale(None=OneMinus(None=@n95)) {bias=0, scale=1}
   B=ScalarParameter() {defaultValue=1, group=Embers, parameterName=EmberBoost2}
  )
  n107 = Multiply(A=Saturate(None=Multiply(A=@n105, B=@n106)), B=Multiply(A=NamedRerouteUsage()) {constB=2})
  n108 = Add(A=LinearInterpolate(A=@n92.RGB, B=@n93.RGB, Alpha=Saturate(None=@n100)), B=@n107)
  n109 = Add(A=@n2, B=LinearInterpolate(A=@n3, B=Multiply(A=@n3) {constB=0.1}, Alpha=@n7.Linear Sine)) {constB=0.01}
  n110 = Max(
   A=Multiply(
    A=@n96
    B=MaterialFunctionCall(Alpha=@n39.Result, Min=@n2, Max=@n109) {desc=Flame Edge, materialFunction=&SmoothStep}.Result
   )
   B=@n107
  )
  NamedRerouteDeclaration(None=Multiply(A=LinearInterpolate(A=@n97, B=@n108, Alpha=@n99.R), B=@n110)) {name=V_Blade_Fill_Outline_Color, nodeColor=<rgba 0 0 0 1>}
  NamedRerouteDeclaration(None=@n110) {name=V_Blade_Fill_Outline_Mask, nodeColor=<rgba 0 0 0 1>}
  NamedRerouteDeclaration(None=Multiply(A=@n82, B=@n91) {desc=Blade Fill}) {name=V_Blade_Fill, nodeColor=<rgba 0 0 0 1>}
  n111 = OneMinus(None=Multiply(A=@n90) {constB=0.66})
  n112 = MaterialFunctionCall(UVs=Add(A=@n86.Flipped Symmetry, B=Constant2Vector() {g=0.008, r=0.39}), Texture Scale=@n85) {materialFunction=&ScaleUVsByCenter}
  n113 = Constant2Vector() {g=0.5, r=1}
  n114 = OneMinus(None=MaterialFunctionCall(
    UVs=MaterialFunctionCall(UV=@n112.UVs, TextureSize=@n113) {materialFunction=&MF_UI_UVScaling}.Scaled to Fit
    Size=Constant() {r=1.74}
   ) {materialFunction=&MF_UI_Circle}.SDF)
  n115 = Multiply(
   A=MaterialFunctionCall(
    DistanceField=MaterialFunctionCall(DistanceFieldA=@n111, DistanceFieldB=@n114) {desc=Blade Edge SDF, materialFunction=&MF_UI_SDF_Intersection}.Exact
    Threshold=Constant() {r=0.882}
    Smoothing=Constant()
   ) {desc=Blade Edge Silhouette, materialFunction=&MF_UI_AntiAliasing}.Inner
   B=@n91
  )
  NamedRerouteDeclaration(None=Multiply(A=@n82, B=NamedRerouteDeclaration(None=@n115) {name=V_Blade_Edge, nodeColor=<rgba 0 0 0 1>}) {desc=Blade Edge Fill}) {name=V_Blade_Edge_Fill, nodeColor=<rgba 0 0 0 1>}
  n116 = ScalarParameter() {defaultValue=0.8, parameterName="Blade (Fill) - Glow"}
  NamedRerouteDeclaration(None=Divide(A=@n90, B=LinearInterpolate(Alpha=@n116) {constA=40}) {constB=1, desc="Blade Glow, Raw"}) {name=V_Blade_Glow_Raw, nodeColor=<rgba 0 0 0 1>}
  Reroute(None=@n56)
  TextureCoordinate()
 ]
}
